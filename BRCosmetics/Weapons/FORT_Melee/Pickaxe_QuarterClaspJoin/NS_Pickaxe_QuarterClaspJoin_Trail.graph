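# Niagara ribbon swing trail for the QuarterClaspJoin pickaxe, drawn between two bone or socket indices as short-lived anim trails.
# Gameplay and front-end emitters gate spawn count by context, while a glow emitter runs in both.
# Time-of-day curves tint the trails and scale gameplay alpha, with glow using fixed white and alpha in the front end.
# The dimmer 001 copies are active, and the brighter original emitters are disabled.
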NiagaraSystem /BRCosmetics/Weapons/FORT_Melee/Pickaxe_QuarterClaspJoin/NS_Pickaxe_QuarterClaspJoin_Trail
dependencies {
 usedDataInterfaces [&NiagaraDataInterfaceCurve, &NiagaraDataInterfaceColorCurve, &NiagaraDataInterfaceSocketReader]
 usedDynamicInputs []
 usedModules [
  &self:SystemSpawnScript.SystemScriptSource.SystemScriptGraph.NiagaraNodeAssignment_1.SetVariables_#1
  &SystemState
  &self:SystemSpawnScript.SystemScriptSource.SystemScriptGraph.NiagaraNodeAssignment_2.SetVariables_#2
 ]
}
emitters {
 FN_AnimTrailGameplay {
  bEnabled=false
  emitterSpawnScript [&FN_RibbonIDOffsetUtility()]
  emitterUpdateScript [
   &EmitterState()
   &FN_SpawnAnimTrail(Number of Anim Trails:NiagaraInt32=&MakeIntFromFloat(Float=&FN_FrontEndFloatModifier(Gameplay=1, Front End=0), Conversion Mode=::NewEnumerator1))
  ]
  particleSpawnScript [
   &FN_InitializeAnimTrailBoneIndecies(First Bone or Socket Index=1, Second Bone or Socket Index=0)
   &FN_InitializeAnimTrailBoneIndecies(AnimTrail Selection=1, First Bone or Socket Index=2, Second Bone or Socket Index=3) {enabled=false, module=FN_InitializeAnimTrailBoneIndecies001}
   &FN_InitializeAnimTrail(AnimTrailWidth=1.4, Lifetime=0.13)
   &AddVelocity(Velocity=<0 400 0>) {enabled=false}
  ]
  particleUpdateScript [
   &SolveForcesAndVelocity(Clamp Velocity=false, Limit Acceleration=false)
   &FN_UpdateAnimTrail() {module=FN_UpdateAnimTrail001}
   &DynamicMaterialParameters(
    Index 1 Param 3=&FloatFromCurve(
     FloatCurve=DataInterface() {propertyValues {Curve {keys [<at 0 0>, {arriveTangent=1.9814, tangentMode=RCTM_User, time=1, value=1}]}}}
     CurveIndex=&ModuloFloat(A=$Emitter.Age, B=1)
     Scale Curve=0
    )
    Index 1 Param 4=1
    Index 2 Param 1=10
    Index 2 Param 2=1
    Index 2 Param 3=0.5
    Index 2 Param 4=0.6
    Module.Index0Float0Write=true
    Module.Index0Float1Write=true
    Module.Index0Float2Write=true
    Module.Index0Float3Write=true
    Module.Index1Float0Write=true
    Module.Index1Float1Write=true
    Module.Index1Float2Write=true
    Module.Index1Float3Write=true
    Module.Index2Float0Write=true
    Module.Index2Float1Write=true
    Module.Index2Float2Write=true
    Module.Index2Float3Write=true
    Write Parameter Index 1=::NewEnumerator1
    Write Parameter Index 2=::NewEnumerator1
   )
   &ScaleColor(
    Scale Alpha=&FloatFromCurve(
     FloatCurve=DataInterface() {propertyValues {Curve {keys [
         {arriveTangent=-3.0232, interpMode=RCIM_Cubic, leaveTangent=-3.0232, tangentMode=RCTM_User, time=0, value=1.5}
         {arriveTangent=-0.2795, tangentMode=RCTM_User, time=1, value=0}
        ]}}}
     Scale Curve=&FN_Select_Float_from_Time_Of_Day(Morning=0.8, Day=1.5, Evening=0.8, Night=0.9, Morning Threshold=3)
    )
    Scale RGB=&MakeVectorFromLinearColorRGB(Color=&ColorFromCurve(
      ColorCurve=DataInterface() {propertyValues {
        AlphaCurve {keys [<at 0 1 RCIM_Cubic>]}
        BlueCurve {keys [
          <at 0 1>
          {arriveTangent=-3.4789, interpMode=RCIM_Cubic, leaveTangent=-3.4789, time=0.0932, value=0.64}
          <at 0.1437 0.5>
          {arriveTangent=1.2068, interpMode=RCIM_Cubic, leaveTangent=1.2068, time=0.3527, value=1}
          {arriveTangent=-1.0717, interpMode=RCIM_Cubic, leaveTangent=-1.0717, time=0.558, value=1}
          {arriveTangent=-0.299, interpMode=RCIM_Cubic, leaveTangent=-0.299, time=0.8165, value=0.5029}
          <at 0.905 0.8963>
         ]}
        GreenCurve {keys [
          <at 0 0.675>
          {arriveTangent=-0.3479, interpMode=RCIM_Cubic, leaveTangent=-0.3479, time=0.0932, value=0.6533}
          <at 0.1437 0.625>
          {arriveTangent=0.9051, interpMode=RCIM_Cubic, leaveTangent=0.9051, time=0.3527, value=1}
          {arriveTangent=-0.8037, interpMode=RCIM_Cubic, leaveTangent=-0.8037, time=0.558, value=1}
          {arriveTangent=-1.1226, interpMode=RCIM_Cubic, leaveTangent=-1.1226, time=0.8165, value=0.6272}
          <at 0.905 0.6105>
         ]}
        RedCurve {keys [
          <at 0 0.35>
          {arriveTangent=4.5225, interpMode=RCIM_Cubic, leaveTangent=4.5225, time=0.0932, value=0.8}
          <at 0.1437 1>
          <at 0.3527 1 RCIM_Cubic>
          <at 0.558 1 RCIM_Cubic>
          {arriveTangent=-1.9772, interpMode=RCIM_Cubic, leaveTangent=-1.9772, time=0.8165, value=1}
          <at 0.905 0.314>
         ]}
       }}
      CurveIndex=Linked() {linkedVariable {name=NPC.FortniteNPC.RemappedTimeOfDay_0-1, type=NiagaraFloat}}
     ))
    ScaleA=true
    ScaleRGB=true
   )
  ]
  rendererClasses [&NiagaraRibbonRendererProperties]
  renderers [
   {rendererClass=&NiagaraRibbonRendererProperties}
   {rendererClass=&NiagaraRibbonRendererProperties, rendererIndex=1}
  ]
 }
 FN_AnimTrailGameplay001 {
  emitterSpawnScript [&FN_RibbonIDOffsetUtility()]
  emitterUpdateScript [
   &EmitterState()
   &FN_SpawnAnimTrail(Number of Anim Trails:NiagaraInt32=&MakeIntFromFloat(Float=&FN_FrontEndFloatModifier(Gameplay=1, Front End=0), Conversion Mode=::NewEnumerator1))
  ]
  particleSpawnScript [
   &FN_InitializeAnimTrailBoneIndecies(First Bone or Socket Index=1, Second Bone or Socket Index=0)
   &FN_InitializeAnimTrailBoneIndecies(AnimTrail Selection=1, First Bone or Socket Index=2, Second Bone or Socket Index=3) {enabled=false, module=FN_InitializeAnimTrailBoneIndecies001}
   &FN_InitializeAnimTrail(AnimTrailWidth=1.4, Lifetime=0.13)
   &AddVelocity(Velocity=<0 400 0>) {enabled=false}
  ]
  particleUpdateScript [
   &SolveForcesAndVelocity(Clamp Velocity=false, Limit Acceleration=false)
   &FN_UpdateAnimTrail() {module=FN_UpdateAnimTrail001}
   &DynamicMaterialParameters(
    Index 1 Param 3=&FloatFromCurve(
     FloatCurve=DataInterface() {propertyValues {Curve {keys [<at 0 0>, {arriveTangent=1.9814, tangentMode=RCTM_User, time=1, value=1}]}}}
     CurveIndex=&ModuloFloat(A=$Emitter.Age, B=1)
     Scale Curve=0
    )
    Index 1 Param 4=1
    Index 2 Param 1=10
    Index 2 Param 2=1
    Index 2 Param 3=0.5
    Index 2 Param 4=0.6
    Module.Index0Float0Write=true
    Module.Index0Float1Write=true
    Module.Index0Float2Write=true
    Module.Index0Float3Write=true
    Module.Index1Float0Write=true
    Module.Index1Float1Write=true
    Module.Index1Float2Write=true
    Module.Index1Float3Write=true
    Module.Index2Float0Write=true
    Module.Index2Float1Write=true
    Module.Index2Float2Write=true
    Module.Index2Float3Write=true
    Write Parameter Index 1=::NewEnumerator1
    Write Parameter Index 2=::NewEnumerator1
   )
   &ScaleColor(
    Scale Alpha=&FloatFromCurve(
     FloatCurve=DataInterface() {propertyValues {Curve {keys [
         {arriveTangent=-3.0232, interpMode=RCIM_Cubic, leaveTangent=-3.0232, tangentMode=RCTM_User, time=0, value=1.5}
         {arriveTangent=-0.2795, tangentMode=RCTM_User, time=1, value=0}
        ]}}}
     Scale Curve=&FN_Select_Float_from_Time_Of_Day(Morning=0.3, Day=0.5, Evening=0.3, Night=0.32, Morning Threshold=3)
    )
    Scale RGB=&MakeVectorFromLinearColorRGB(Color=&ColorFromCurve(
      ColorCurve=DataInterface() {propertyValues {
        AlphaCurve {keys [<at 0 1 RCIM_Cubic>]}
        BlueCurve {keys [
          <at 0 1>
          {arriveTangent=-3.4789, interpMode=RCIM_Cubic, leaveTangent=-3.4789, time=0.0932, value=0.64}
          <at 0.1437 0.5>
          {arriveTangent=1.2068, interpMode=RCIM_Cubic, leaveTangent=1.2068, time=0.3527, value=1}
          {arriveTangent=-1.0717, interpMode=RCIM_Cubic, leaveTangent=-1.0717, time=0.558, value=1}
          {arriveTangent=-0.299, interpMode=RCIM_Cubic, leaveTangent=-0.299, time=0.8165, value=0.5029}
          <at 0.905 0.8963>
         ]}
        GreenCurve {keys [
          <at 0 0.675>
          {arriveTangent=-0.3479, interpMode=RCIM_Cubic, leaveTangent=-0.3479, time=0.0932, value=0.6533}
          <at 0.1437 0.625>
          {arriveTangent=0.9051, interpMode=RCIM_Cubic, leaveTangent=0.9051, time=0.3527, value=1}
          {arriveTangent=-0.8037, interpMode=RCIM_Cubic, leaveTangent=-0.8037, time=0.558, value=1}
          {arriveTangent=-1.1226, interpMode=RCIM_Cubic, leaveTangent=-1.1226, time=0.8165, value=0.6272}
          <at 0.905 0.6105>
         ]}
        RedCurve {keys [
          <at 0 0.35>
          {arriveTangent=4.5225, interpMode=RCIM_Cubic, leaveTangent=4.5225, time=0.0932, value=0.8}
          <at 0.1437 1>
          <at 0.3527 1 RCIM_Cubic>
          <at 0.558 1 RCIM_Cubic>
          {arriveTangent=-1.9772, interpMode=RCIM_Cubic, leaveTangent=-1.9772, time=0.8165, value=1}
          <at 0.905 0.314>
         ]}
       }}
      CurveIndex=Linked() {linkedVariable {name=NPC.FortniteNPC.RemappedTimeOfDay_0-1, type=NiagaraFloat}}
     ))
    ScaleA=true
    ScaleRGB=true
   )
  ]
  rendererClasses [&NiagaraRibbonRendererProperties]
  renderers [
   {rendererClass=&NiagaraRibbonRendererProperties}
   {rendererClass=&NiagaraRibbonRendererProperties, rendererIndex=1}
  ]
 }
 FN_AnimTrail_FE {
  bEnabled=false
  emitterSpawnScript [&FN_RibbonIDOffsetUtility()]
  emitterUpdateScript [
   &EmitterState()
   &FN_SpawnAnimTrail(Number of Anim Trails:NiagaraInt32=&MakeIntFromFloat(Float=&FN_FrontEndFloatModifier(Gameplay=0, Front End=1), Conversion Mode=::NewEnumerator1))
  ]
  particleSpawnScript [
   &FN_InitializeAnimTrailBoneIndecies(First Bone or Socket Index=1, Second Bone or Socket Index=0)
   &FN_InitializeAnimTrailBoneIndecies(AnimTrail Selection=1, First Bone or Socket Index=2, Second Bone or Socket Index=3) {enabled=false, module=FN_InitializeAnimTrailBoneIndecies001}
   &FN_InitializeAnimTrail(AnimTrailWidth=1.4, Lifetime=0.13)
   &AddVelocity(Velocity=<0 400 0>) {enabled=false}
  ]
  particleUpdateScript [
   &SolveForcesAndVelocity(Clamp Velocity=false, Limit Acceleration=false)
   &FN_UpdateAnimTrail() {module=FN_UpdateAnimTrail001}
   &DynamicMaterialParameters(
    Index 1 Param 3=&FloatFromCurve(
     FloatCurve=DataInterface() {propertyValues {Curve {keys [<at 0 0>, {arriveTangent=1.9814, tangentMode=RCTM_User, time=1, value=1}]}}}
     CurveIndex=&ModuloFloat(A=$Emitter.Age, B=1)
     Scale Curve=0
    )
    Index 1 Param 4=1
    Index 2 Param 1=10
    Index 2 Param 2=1
    Index 2 Param 3=0.5
    Index 2 Param 4=0.6
    Module.Index0Float0Write=true
    Module.Index0Float1Write=true
    Module.Index0Float2Write=true
    Module.Index0Float3Write=true
    Module.Index1Float0Write=true
    Module.Index1Float1Write=true
    Module.Index1Float2Write=true
    Module.Index1Float3Write=true
    Module.Index2Float0Write=true
    Module.Index2Float1Write=true
    Module.Index2Float2Write=true
    Module.Index2Float3Write=true
    Write Parameter Index 1=::NewEnumerator1
    Write Parameter Index 2=::NewEnumerator1
   )
   &ScaleColor(
    Scale Alpha=&FloatFromCurve(
     FloatCurve=DataInterface() {propertyValues {Curve {keys [
         {arriveTangent=-3.0232, interpMode=RCIM_Cubic, leaveTangent=-3.0232, tangentMode=RCTM_User, time=0, value=1.5}
         {arriveTangent=-0.2795, tangentMode=RCTM_User, time=1, value=0}
        ]}}}
     Scale Curve=1.2
    )
    Scale RGB=&MakeVectorFromLinearColorRGB(Color=&ColorFromCurve(
      ColorCurve=DataInterface() {propertyValues {
        AlphaCurve {keys [<at 0 1 RCIM_Cubic>]}
        BlueCurve {keys [
          <at 0 1>
          {arriveTangent=-3.4789, interpMode=RCIM_Cubic, leaveTangent=-3.4789, time=0.0932, value=0.64}
          <at 0.1437 0.5>
          {arriveTangent=1.2068, interpMode=RCIM_Cubic, leaveTangent=1.2068, time=0.3527, value=1}
          {arriveTangent=-1.0717, interpMode=RCIM_Cubic, leaveTangent=-1.0717, time=0.558, value=1}
          {arriveTangent=-0.299, interpMode=RCIM_Cubic, leaveTangent=-0.299, time=0.8165, value=0.5029}
          <at 0.905 0.8963>
         ]}
        GreenCurve {keys [
          <at 0 0.675>
          {arriveTangent=-0.3479, interpMode=RCIM_Cubic, leaveTangent=-0.3479, time=0.0932, value=0.6533}
          <at 0.1437 0.625>
          {arriveTangent=0.9051, interpMode=RCIM_Cubic, leaveTangent=0.9051, time=0.3527, value=1}
          {arriveTangent=-0.8037, interpMode=RCIM_Cubic, leaveTangent=-0.8037, time=0.558, value=1}
          {arriveTangent=-1.1226, interpMode=RCIM_Cubic, leaveTangent=-1.1226, time=0.8165, value=0.6272}
          <at 0.905 0.6105>
         ]}
        RedCurve {keys [
          <at 0 0.35>
          {arriveTangent=4.5225, interpMode=RCIM_Cubic, leaveTangent=4.5225, time=0.0932, value=0.8}
          <at 0.1437 1>
          <at 0.3527 1 RCIM_Cubic>
          <at 0.558 1 RCIM_Cubic>
          {arriveTangent=-1.9772, interpMode=RCIM_Cubic, leaveTangent=-1.9772, time=0.8165, value=1}
          <at 0.905 0.314>
         ]}
       }}
      CurveIndex=Linked() {linkedVariable {name=NPC.FortniteNPC.RemappedTimeOfDay_0-1, type=NiagaraFloat}}
     ))
    ScaleA=true
    ScaleRGB=true
   )
  ]
  rendererClasses [&NiagaraRibbonRendererProperties]
  renderers [
   {rendererClass=&NiagaraRibbonRendererProperties}
   {rendererClass=&NiagaraRibbonRendererProperties, rendererIndex=1}
  ]
 }
 FN_AnimTrail_FE001 {
  emitterSpawnScript [&FN_RibbonIDOffsetUtility()]
  emitterUpdateScript [
   &EmitterState()
   &FN_SpawnAnimTrail(Number of Anim Trails:NiagaraInt32=&MakeIntFromFloat(Float=&FN_FrontEndFloatModifier(Gameplay=0, Front End=1), Conversion Mode=::NewEnumerator1))
  ]
  particleSpawnScript [
   &FN_InitializeAnimTrailBoneIndecies(First Bone or Socket Index=1, Second Bone or Socket Index=0)
   &FN_InitializeAnimTrailBoneIndecies(AnimTrail Selection=1, First Bone or Socket Index=2, Second Bone or Socket Index=3) {enabled=false, module=FN_InitializeAnimTrailBoneIndecies001}
   &FN_InitializeAnimTrail(AnimTrailWidth=1.4, Lifetime=0.13)
   &AddVelocity(Velocity=<0 400 0>) {enabled=false}
  ]
  particleUpdateScript [
   &SolveForcesAndVelocity(Clamp Velocity=false, Limit Acceleration=false)
   &FN_UpdateAnimTrail() {module=FN_UpdateAnimTrail001}
   &DynamicMaterialParameters(
    Index 1 Param 3=&FloatFromCurve(
     FloatCurve=DataInterface() {propertyValues {Curve {keys [<at 0 0>, {arriveTangent=1.9814, tangentMode=RCTM_User, time=1, value=1}]}}}
     CurveIndex=&ModuloFloat(A=$Emitter.Age, B=1)
     Scale Curve=0
    )
    Index 1 Param 4=1
    Index 2 Param 1=10
    Index 2 Param 2=1
    Index 2 Param 3=0.5
    Index 2 Param 4=0.6
    Module.Index0Float0Write=true
    Module.Index0Float1Write=true
    Module.Index0Float2Write=true
    Module.Index0Float3Write=true
    Module.Index1Float0Write=true
    Module.Index1Float1Write=true
    Module.Index1Float2Write=true
    Module.Index1Float3Write=true
    Module.Index2Float0Write=true
    Module.Index2Float1Write=true
    Module.Index2Float2Write=true
    Module.Index2Float3Write=true
    Write Parameter Index 1=::NewEnumerator1
    Write Parameter Index 2=::NewEnumerator1
   )
   &ScaleColor(
    Scale Alpha=&FloatFromCurve(
     FloatCurve=DataInterface() {propertyValues {Curve {keys [
         {arriveTangent=-3.0232, interpMode=RCIM_Cubic, leaveTangent=-3.0232, tangentMode=RCTM_User, time=0, value=1.5}
         {arriveTangent=-0.2795, tangentMode=RCTM_User, time=1, value=0}
        ]}}}
     Scale Curve=0.2
    )
    Scale RGB=&MakeVectorFromLinearColorRGB(Color=&ColorFromCurve(
      ColorCurve=DataInterface() {propertyValues {
        AlphaCurve {keys [<at 0 1 RCIM_Cubic>]}
        BlueCurve {keys [
          <at 0 1>
          {arriveTangent=-3.4789, interpMode=RCIM_Cubic, leaveTangent=-3.4789, time=0.0932, value=0.64}
          <at 0.1437 0.5>
          {arriveTangent=1.2068, interpMode=RCIM_Cubic, leaveTangent=1.2068, time=0.3527, value=1}
          {arriveTangent=-1.0717, interpMode=RCIM_Cubic, leaveTangent=-1.0717, time=0.558, value=1}
          {arriveTangent=-0.299, interpMode=RCIM_Cubic, leaveTangent=-0.299, time=0.8165, value=0.5029}
          <at 0.905 0.8963>
         ]}
        GreenCurve {keys [
          <at 0 0.675>
          {arriveTangent=-0.3479, interpMode=RCIM_Cubic, leaveTangent=-0.3479, time=0.0932, value=0.6533}
          <at 0.1437 0.625>
          {arriveTangent=0.9051, interpMode=RCIM_Cubic, leaveTangent=0.9051, time=0.3527, value=1}
          {arriveTangent=-0.8037, interpMode=RCIM_Cubic, leaveTangent=-0.8037, time=0.558, value=1}
          {arriveTangent=-1.1226, interpMode=RCIM_Cubic, leaveTangent=-1.1226, time=0.8165, value=0.6272}
          <at 0.905 0.6105>
         ]}
        RedCurve {keys [
          <at 0 0.35>
          {arriveTangent=4.5225, interpMode=RCIM_Cubic, leaveTangent=4.5225, time=0.0932, value=0.8}
          <at 0.1437 1>
          <at 0.3527 1 RCIM_Cubic>
          <at 0.558 1 RCIM_Cubic>
          {arriveTangent=-1.9772, interpMode=RCIM_Cubic, leaveTangent=-1.9772, time=0.8165, value=1}
          <at 0.905 0.314>
         ]}
       }}
      CurveIndex=Linked() {linkedVariable {name=NPC.FortniteNPC.RemappedTimeOfDay_0-1, type=NiagaraFloat}}
     ))
    ScaleA=true
    ScaleRGB=true
   )
  ]
  rendererClasses [&NiagaraRibbonRendererProperties]
  renderers [
   {rendererClass=&NiagaraRibbonRendererProperties}
   {rendererClass=&NiagaraRibbonRendererProperties, rendererIndex=1}
  ]
 }
 FN_AnimTrail_Glow {
  bEnabled=false
  emitterSpawnScript [&FN_RibbonIDOffsetUtility()]
  emitterUpdateScript [
   &EmitterState()
   &FN_SpawnAnimTrail()
  ]
  particleSpawnScript [
   &FN_InitializeAnimTrailBoneIndecies(First Bone or Socket Index=1, Second Bone or Socket Index=0)
   &FN_InitializeAnimTrailBoneIndecies(AnimTrail Selection=1, First Bone or Socket Index=2, Second Bone or Socket Index=3) {enabled=false, module=FN_InitializeAnimTrailBoneIndecies001}
   &FN_InitializeAnimTrail(AnimTrailWidth=1.5, Lifetime=0.12)
   &AddVelocity(Velocity=<0 400 0>) {enabled=false}
  ]
  particleUpdateScript [
   &SolveForcesAndVelocity(Clamp Velocity=false, Limit Acceleration=false)
   &FN_UpdateAnimTrail() {module=FN_UpdateAnimTrail001}
   &DynamicMaterialParameters(
    Index 1 Param 3=&FloatFromCurve(
     FloatCurve=DataInterface() {propertyValues {Curve {keys [<at 0 0>, {arriveTangent=1.9814, tangentMode=RCTM_User, time=1, value=1}]}}}
     CurveIndex=&ModuloFloat(A=$Emitter.Age, B=1)
     Scale Curve=0
    )
    Index 1 Param 4=1
    Index 2 Param 1=10
    Index 2 Param 2=1
    Index 2 Param 3=0.5
    Index 2 Param 4=0.6
    Module.Index0Float0Write=true
    Module.Index0Float1Write=true
    Module.Index0Float2Write=true
    Module.Index0Float3Write=true
    Module.Index1Float0Write=true
    Module.Index1Float1Write=true
    Module.Index1Float2Write=true
    Module.Index1Float3Write=true
    Module.Index2Float0Write=true
    Module.Index2Float1Write=true
    Module.Index2Float2Write=true
    Module.Index2Float3Write=true
    Write Parameter Index 1=::NewEnumerator1
    Write Parameter Index 2=::NewEnumerator1
   )
   &ScaleColor(
    Scale Alpha=&FN_FrontEndFloatModifier(
     Gameplay=&FloatFromCurve(
      FloatCurve=DataInterface() {propertyValues {Curve {keys [
          {arriveTangent=-3.0232, interpMode=RCIM_Cubic, leaveTangent=-3.0232, tangentMode=RCTM_User, time=0, value=1}
          {arriveTangent=-0.2795, tangentMode=RCTM_User, time=1, value=0}
         ]}}}
      Scale Curve=&FN_Select_Float_from_Time_Of_Day(Morning=1, Day=1.15, Evening=1, Night=1.15, Morning Threshold=3)
     )
     Front End=0.15
    )
    Scale RGB=&FN_FrontEndVector3Modifier(
     Gameplay=&MakeVectorFromLinearColorRGB(Color=&ColorFromCurve(
       ColorCurve=DataInterface() {propertyValues {
         AlphaCurve {keys [<at 0 1 RCIM_Cubic>]}
         BlueCurve {keys [
           <at 0 0.75>
           {arriveTangent=-1.7394, interpMode=RCIM_Cubic, leaveTangent=-1.7394, time=0.0932, value=0.5408}
           <at 0.1437 0.5>
           {arriveTangent=1.2068, interpMode=RCIM_Cubic, leaveTangent=1.2068, time=0.3527, value=1}
           {arriveTangent=-1.0717, interpMode=RCIM_Cubic, leaveTangent=-1.0717, time=0.558, value=1}
           {arriveTangent=-0.712, interpMode=RCIM_Cubic, leaveTangent=-0.712, time=0.8165, value=0.5029}
           <at 0.905 0.7529>
          ]}
         GreenCurve {keys [
           <at 0 0.5062>
           {arriveTangent=1.696, interpMode=RCIM_Cubic, leaveTangent=1.696, time=0.0932, value=0.5624}
           <at 0.1437 0.75>
           {arriveTangent=0.6034, interpMode=RCIM_Cubic, leaveTangent=0.6034, time=0.3527, value=1}
           {arriveTangent=-0.5326, interpMode=RCIM_Cubic, leaveTangent=-0.5326, time=0.558, value=1}
           {arriveTangent=-1.4155, interpMode=RCIM_Cubic, leaveTangent=-1.4155, time=0.8165, value=0.7529}
           <at 0.905 0.5089>
          ]}
         RedCurve {keys [
           <at 0 0.2625>
           {arriveTangent=5.1313, interpMode=RCIM_Cubic, leaveTangent=5.1313, time=0.0932, value=0.8}
           <at 0.1437 1>
           <at 0.3527 1 RCIM_Cubic>
           <at 0.558 1 RCIM_Cubic>
           {arriveTangent=-2.1263, interpMode=RCIM_Cubic, leaveTangent=-2.1263, time=0.8165, value=1}
           <at 0.905 0.2623>
          ]}
        }}
       CurveIndex=Linked() {linkedVariable {name=NPC.FortniteNPC.RemappedTimeOfDay_0-1, type=NiagaraFloat}}
      ))
     Front End=<1 1 1>
    )
    ScaleA=true
    ScaleRGB=true
   )
  ]
  rendererClasses [&NiagaraRibbonRendererProperties]
  renderers [
   {rendererClass=&NiagaraRibbonRendererProperties}
   {rendererClass=&NiagaraRibbonRendererProperties, rendererIndex=1}
  ]
 }
 FN_AnimTrail_Glow001 {
  emitterSpawnScript [&FN_RibbonIDOffsetUtility()]
  emitterUpdateScript [
   &EmitterState()
   &FN_SpawnAnimTrail()
  ]
  particleSpawnScript [
   &FN_InitializeAnimTrailBoneIndecies(First Bone or Socket Index=1, Second Bone or Socket Index=0)
   &FN_InitializeAnimTrailBoneIndecies(AnimTrail Selection=1, First Bone or Socket Index=2, Second Bone or Socket Index=3) {enabled=false, module=FN_InitializeAnimTrailBoneIndecies001}
   &FN_InitializeAnimTrail(AnimTrailWidth=1.5, Lifetime=0.12)
   &AddVelocity(Velocity=<0 400 0>) {enabled=false}
  ]
  particleUpdateScript [
   &SolveForcesAndVelocity(Clamp Velocity=false, Limit Acceleration=false)
   &FN_UpdateAnimTrail() {module=FN_UpdateAnimTrail001}
   &DynamicMaterialParameters(
    Index 1 Param 3=&FloatFromCurve(
     FloatCurve=DataInterface() {propertyValues {Curve {keys [<at 0 0>, {arriveTangent=1.9814, tangentMode=RCTM_User, time=1, value=1}]}}}
     CurveIndex=&ModuloFloat(A=$Emitter.Age, B=1)
     Scale Curve=0
    )
    Index 1 Param 4=1
    Index 2 Param 1=10
    Index 2 Param 2=1
    Index 2 Param 3=0.5
    Index 2 Param 4=0.6
    Module.Index0Float0Write=true
    Module.Index0Float1Write=true
    Module.Index0Float2Write=true
    Module.Index0Float3Write=true
    Module.Index1Float0Write=true
    Module.Index1Float1Write=true
    Module.Index1Float2Write=true
    Module.Index1Float3Write=true
    Module.Index2Float0Write=true
    Module.Index2Float1Write=true
    Module.Index2Float2Write=true
    Module.Index2Float3Write=true
    Write Parameter Index 1=::NewEnumerator1
    Write Parameter Index 2=::NewEnumerator1
   )
   &ScaleColor(
    Scale Alpha=&FN_FrontEndFloatModifier(
     Gameplay=&FloatFromCurve(
      FloatCurve=DataInterface() {propertyValues {Curve {keys [
          {arriveTangent=-3.0232, interpMode=RCIM_Cubic, leaveTangent=-3.0232, tangentMode=RCTM_User, time=0, value=1}
          {arriveTangent=-0.2795, tangentMode=RCTM_User, time=1, value=0}
         ]}}}
      Scale Curve=&FN_Select_Float_from_Time_Of_Day(Morning=0.08, Day=0.1, Evening=0.08, Night=0.1, Morning Threshold=3)
     )
     Front End=0.035
    )
    Scale RGB=&FN_FrontEndVector3Modifier(
     Gameplay=&MakeVectorFromLinearColorRGB(Color=&ColorFromCurve(
       ColorCurve=DataInterface() {propertyValues {
         AlphaCurve {keys [<at 0 1 RCIM_Cubic>]}
         BlueCurve {keys [
           <at 0 0.75>
           {arriveTangent=-1.7394, interpMode=RCIM_Cubic, leaveTangent=-1.7394, time=0.0932, value=0.5408}
           <at 0.1437 0.5>
           {arriveTangent=1.2068, interpMode=RCIM_Cubic, leaveTangent=1.2068, time=0.3527, value=1}
           {arriveTangent=-1.0717, interpMode=RCIM_Cubic, leaveTangent=-1.0717, time=0.558, value=1}
           {arriveTangent=-0.712, interpMode=RCIM_Cubic, leaveTangent=-0.712, time=0.8165, value=0.5029}
           <at 0.905 0.7529>
          ]}
         GreenCurve {keys [
           <at 0 0.5062>
           {arriveTangent=1.696, interpMode=RCIM_Cubic, leaveTangent=1.696, time=0.0932, value=0.5624}
           <at 0.1437 0.75>
           {arriveTangent=0.6034, interpMode=RCIM_Cubic, leaveTangent=0.6034, time=0.3527, value=1}
           {arriveTangent=-0.5326, interpMode=RCIM_Cubic, leaveTangent=-0.5326, time=0.558, value=1}
           {arriveTangent=-1.4155, interpMode=RCIM_Cubic, leaveTangent=-1.4155, time=0.8165, value=0.7529}
           <at 0.905 0.5089>
          ]}
         RedCurve {keys [
           <at 0 0.2625>
           {arriveTangent=5.1313, interpMode=RCIM_Cubic, leaveTangent=5.1313, time=0.0932, value=0.8}
           <at 0.1437 1>
           <at 0.3527 1 RCIM_Cubic>
           <at 0.558 1 RCIM_Cubic>
           {arriveTangent=-2.1263, interpMode=RCIM_Cubic, leaveTangent=-2.1263, time=0.8165, value=1}
           <at 0.905 0.2623>
          ]}
        }}
       CurveIndex=Linked() {linkedVariable {name=NPC.FortniteNPC.RemappedTimeOfDay_0-1, type=NiagaraFloat}}
      ))
     Front End=<1 1 1>
    )
    ScaleA=true
    ScaleRGB=true
   )
  ]
  rendererClasses [&NiagaraRibbonRendererProperties]
  renderers [
   {rendererClass=&NiagaraRibbonRendererProperties}
   {rendererClass=&NiagaraRibbonRendererProperties, rendererIndex=1}
  ]
 }
}
summary {
 systemName=NS_Pickaxe_QuarterClaspJoin_Trail
 emitters [
  {bEnabled=false, emitterName=FN_AnimTrailGameplay, rendererClasses [&NiagaraRibbonRendererProperties]}
  {bEnabled=false, emitterName=FN_AnimTrail_Glow, rendererClasses [&NiagaraRibbonRendererProperties]}
  {bEnabled=false, emitterName=FN_AnimTrail_FE, rendererClasses [&NiagaraRibbonRendererProperties]}
  {emitterName=FN_AnimTrailGameplay001, rendererClasses [&NiagaraRibbonRendererProperties]}
  {emitterName=FN_AnimTrail_Glow001, rendererClasses [&NiagaraRibbonRendererProperties]}
  {emitterName=FN_AnimTrail_FE001, rendererClasses [&NiagaraRibbonRendererProperties]}
 ]
 userVariables [
  User.bIsFrontEnd {type=NiagaraBool}
  User.bIsFrontEndPreview {type=NiagaraBool}
  User.PhysicalSurface {type=NiagaraInt32}
  User.RibbonIDOffset {type=NiagaraInt32}
  User.SwingSectionIndex {type=NiagaraInt32}
 ]
}
